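; Large gate switch Blueprint, invulnerable from BeginPlay, that players interact with to operate a referenced BGA_WallDoor_A.
; Interaction passes through a DoOnce and a door-open check to call CloseGate or OpenGate, with a short delay.
; A DoorState event sets the prompt to Close Gate or Open Gate.
; TheDoorIsMoving blocks interaction, and its OnRep swaps the switch's on/off mesh; this file never sets that flag.
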(event EventBeginPlay
  (Variables|BuildingActor|SetbIsInvulnerable true))

(event Custom|DoorState
  (params (Open :type Boolean))
  (if Open
    (Variables|Default|SetInteractionTextToShow "NSLOCTEXT(\"[52187CEBE30A7E587566912ABE2244D5]\", \"3E06D5F04D5AC53FA23E28BBA58E03B8\", \"Close Gate\")")
    (else
      (Variables|Default|SetInteractionTextToShow "NSLOCTEXT(\"[52187CEBE30A7E587566912ABE2244D5]\", \"A9793A3141161566633154BD1A880C74\", \"Open Gate\")"))))

(event Interaction|EventBlueprintOnInteract
  (params (InteractingPawn :type Fort%20Pawn%20Object%20Reference) (InteractionBeingAttempted :type EInteractionBeingAttempted%20Enum :default "FirstInteraction"))
  (bind _doorref (Variables|Default|GetDoorRef))
  (Utilities|FlowControl|DoOnce false
    (:Completed
      (if (Variables|/Hera/Environment/Prototype/ChrisK/Wall/Door/BGA_WallDoor_A.BGA_WallDoor_A_C|GetDoorIsOpen (Variables|Default|GetDoorRef))
        (Class|BGAWallDoorA|CloseGate _doorref)
        (Utilities|FlowControl|Delay 2.0))
      (Class|BGAWallDoorA|OpenGate _doorref))))

(fn BlueprintCanInteract
  (params (InteractingPawn :type Fort%20Pawn%20Object%20Reference) (InteractionBeingAttempted :type EInteractionBeingAttempted%20Enum :default "FirstInteraction") (InteractionType :type TInteractionType%20Enum :default "IT_NoInteraction"))
  (if (Variables|Default|GetTheDoorIsMoving)
    (return false)
    (else
      (return true))))

(fn BlueprintGetInteractionString
  (params (InteractingPawn :type Fort%20Pawn%20Object%20Reference) (InteractionBeingAttempted :type EInteractionBeingAttempted%20Enum :default "FirstInteraction"))
  (return (Variables|Default|GetInteractionTextToShow)))

(fn OnRep_TheDoorIsMoving
  (if (Variables|Default|GetTheDoorIsMoving)
    (Components|StaticMesh|SetStaticMesh (Variables|Default|GetStaticMesh) "/Game/Creative/Devices/Switch_Device/Meshes/CP_Device_Switch_02_Off.CP_Device_Switch_02_Off")
    (else
      (Components|StaticMesh|SetStaticMesh (Variables|Default|GetStaticMesh) "/Game/Creative/Devices/Switch_Device/Meshes/CP_Device_Switch_02_On.CP_Device_Switch_02_On"))))

(fn ConstructionScript)
