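; Invulnerable building actor for a large three-part wall gate that opens and closes on authority.
; BeginPlay caches start and end offsets so the side panels slide apart on Y and the center panel rises on Z.
; OpenGate and CloseGate play one timeline forward or in reverse to lerp the panels, ignoring calls while moving.
; The linked BGA_LargeDoorSwitch is flagged during movement and receives the final open state when the timeline finishes.
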(event EventBeginPlay
  (bind _bigwalldoor_a_left (Variables|Default|GetBigWallDoor_A_Left))
  (bind _relativelocation (Variables|/Script/Engine.SceneComponent|GetRelativeLocation _bigwalldoor_a_left))
  (bind _bigwalldoor_a_right (Variables|Default|GetBigWallDoor_A_Right))
  (bind _relativelocation_2 (Variables|/Script/Engine.SceneComponent|GetRelativeLocation _bigwalldoor_a_right))
  ; Init
  (Variables|BuildingActor|SetbIsInvulnerable true)
  (Utilities|FlowControl|Sequence
    (:then_0
      (Variables|Default|SetLeftDoorStartRelLoc _relativelocation)
      (Variables|Default|SetLeftDoorEndRelLoc (Math|Vector|MakeVector (.x _relativelocation) (+ (.y _relativelocation) 512.0) (.z _relativelocation))))
    (:then_1
      (Variables|Default|SetRightDoorStartRelLoc _relativelocation_2)
      (Variables|Default|SetRightDoorEndRelLoc (Math|Vector|MakeVector (.x _relativelocation_2) (+ (.y _relativelocation_2) -512.0) (.z _relativelocation_2))))
    (:then_2
      (bind _output_get (Variables|Default|SetCenterDoorStartRelLoc (Variables|/Script/Engine.SceneComponent|GetRelativeLocation (Variables|Default|GetBigWallDoor_B_Center))))
      (Variables|Default|SetCenterDoorEndRelLoc (Math|Vector|MakeVector (.x _output_get) (.y _output_get) (+ (.z _output_get) 3.5e+03))))))

(event Custom|OpenGate
  (bind _open (Variables|BGA_WallDoor_A|GetOpen))
  (bind _doorswitchref (Variables|Default|GetDoorSwitchRef))
  ; Open and close door anim
  (Networking|SwitchHasAuthority
    (:Authority
      (if (not (Variables|Default|GetGateMoving))
        (Variables|/Hera/Environment/Prototype/ChrisK/Wall/Door/BGA_LargeDoorSwitch.BGA_LargeDoorSwitch_C|SetTheDoorIsMoving true (Variables|Default|GetDoorSwitchRef))
        (bind (_direction _lerp) (Timeline|Open PlayFromStart)
          (:Update
            (Variables|Default|SetGateMoving true)
            (Transformation|SetRelativeLocation (Variables|Default|GetBigWallDoor_A_Left) (Math|Vector|Lerp(Vector) (Variables|Default|GetLeftDoorStartRelLoc) (Variables|Default|GetLeftDoorEndRelLoc) _lerp) false false)
            (Transformation|SetRelativeLocation (Variables|Default|GetBigWallDoor_A_Right) (Math|Vector|Lerp(Vector) (Variables|Default|GetRightDoorStartRelLoc) (Variables|Default|GetRightDoorEndRelLoc) _lerp) false false)
            (Transformation|SetRelativeLocation (Variables|Default|GetBigWallDoor_B_Center) (Math|Vector|Lerp(Vector) (Variables|Default|GetCenterDoorStartRelLoc) (Variables|Default|GetCenterDoorEndRelLoc) _lerp) false false))
          (:Finished
            (Variables|Default|SetGateMoving false)
            (bind _output_get (Variables|Default|SetCurrentDoorPlaybackPosition (Components|Timeline|GetPlaybackPosition _open)))
            (if (== _output_get 0)
              (Variables|Default|SetDoorIsOpen false)
              (elif (== _output_get (Components|Timeline|GetTimelineLength _open))
                (Variables|Default|SetDoorIsOpen true)))
            (Variables|/Hera/Environment/Prototype/ChrisK/Wall/Door/BGA_LargeDoorSwitch.BGA_LargeDoorSwitch_C|SetTheDoorIsMoving false (Variables|Default|GetDoorSwitchRef))
            (Utilities|IsValid _doorswitchref
              (:Is%20Valid
                (Class|BGALargeDoorSwitch|DoorState _doorswitchref (Variables|Default|GetDoorIsOpen))))))))))

(event Custom|CloseGate
  (bind _open (Variables|BGA_WallDoor_A|GetOpen))
  (bind _doorswitchref (Variables|Default|GetDoorSwitchRef))
  ; Open and close door anim
  (Networking|SwitchHasAuthority
    (:Authority
      (if (not (Variables|Default|GetGateMoving))
        (Variables|/Hera/Environment/Prototype/ChrisK/Wall/Door/BGA_LargeDoorSwitch.BGA_LargeDoorSwitch_C|SetTheDoorIsMoving true (Variables|Default|GetDoorSwitchRef))
        (Timeline|Open ReverseFromEnd)))))

(fn ConstructionScript)
